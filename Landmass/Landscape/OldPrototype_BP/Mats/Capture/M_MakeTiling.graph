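# Prototype material in the landscape capture folder that, by its name, is meant to make a captured texture tile, but appears unfinished.
# An edge mask from the U axis only (Min takes R twice) lerps half-offset wrapped UVs at tile borders to plain wrapped UVs inside.
# The lerp feeds neither texture parameter (RT, RT_1), and no node is shown connected to a material output.
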Material /Landmass/Landscape/OldPrototype_BP/Mats/Capture/M_MakeTiling
graph {
 properties {pixelDepthOffsetMode=PDOM_Legacy, usedShadingModels=Default Lit}
 nodes [
  Frac(None=Add(A=TextureCoordinate()) {constB=0.5})
  n1 = TextureCoordinate()
  n2 = TextureCoordinate()
  n3 = Abs(None=Subtract(A=Frac(None=@n2)) {constB=0.5})
  n4 = MaterialFunctionCall(Float2=OneMinus(None=Multiply(A=@n3) {constB=2})) {materialFunction=&BreakOutFloat2Components}
  n5 = Min(A=@n4.R, B=@n4.R)
  LinearInterpolate(A=Frac(None=Add(A=@n1) {constB=0.5}), B=Frac(None=TextureCoordinate()), Alpha=Saturate(None=Multiply(A=@n5) {constB=4}))
  TextureSampleParameter2D() {parameterName=RT, texture=null}
  TextureSampleParameter2D(UVs=Add() {constB=0.5}) {parameterName=RT_1, texture=null}
 ]
}
