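# Unlit prototype Landmass capture material that remaps the red channel of a height render-target texture parameter.
# One chain normalizes height between Min and Max parameters and saturates it to 0–1.
# A second chain subtracts a linear ramp from the raw height, built from UVs rotated by an angle parameter around Z and scaled by Ramp.
# Neither chain is shown connected to a material output.
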Material /Landmass/Landscape/OldPrototype_BP/Mats/Capture/M_Normalize_Height
graph {
 properties {pixelDepthOffsetMode=PDOM_Legacy, shadingModel=MSM_Unlit, usedShadingModels=Unlit}
 nodes [
  n1 = TextureSampleParameter2D() {parameterName=RT, texture=&Black}
  n2 = ScalarParameter() {parameterName=Min}
  n3 = ScalarParameter() {defaultValue=10000, parameterName=Max}
  Saturate(None=Divide(A=Subtract(A=@n1.R, B=@n2), B=Subtract(A=@n3, B=@n2)))
  n4 = Constant3Vector() {constant=<rgba 0 0 1 0>}
  n5 = ScalarParameter() {parameterName=ang}
  n6 = Constant() {r=0.5}
  n7 = AppendVector(A=TextureCoordinate(), B=Constant())
  n8 = Add(
   A=ComponentMask(None=RotateAboutAxis(NormalizedRotationAxis=@n4, RotationAngle=@n5, PivotPoint=@n6, Position=@n7)) {g=true, r=true}
   B=TextureCoordinate()
  )
  Subtract(A=@n1.R, B=Multiply(A=ComponentMask(None=@n8) {r=true}, B=ScalarParameter() {defaultValue=1500, parameterName=Ramp}))
 ]
}
