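# Unlit Landmass prototype capture material that, going by its name, normalizes water height from the RT render target into a 0–1 range.
# It samples four neighbors one 1/512 step away and averages their blue values, counting only non-empty ones.
# Where the center blue is positive, it adds that average to the center red; elsewhere it keeps the blue value.
# The result is remapped by Min and Max parameters and saturated.
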Material /Landmass/Landscape/OldPrototype_BP/Mats/Capture/M_Normalize_Height_Water
graph {
 properties {pixelDepthOffsetMode=PDOM_Legacy, shadingModel=MSM_Unlit, usedShadingModels=Unlit}
 nodes [
  n1 = TextureCoordinate()
  n2 = Divide(B=Constant() {r=512})
  n3 = Multiply(A=@n2, B=Constant2Vector() {r=1})
  n4 = TextureSampleParameter2D(UVs=Add(A=@n1, B=@n3)) {parameterName=RT, texture=&Black}
  n5 = TextureSampleParameter2D(UVs=Subtract(A=@n1, B=@n3)) {parameterName=RT, texture=&Black}
  n6 = Multiply(A=@n2, B=Constant2Vector() {g=1})
  n7 = TextureSampleParameter2D(UVs=Add(A=@n1, B=@n6)) {parameterName=RT, texture=&Black}
  n8 = TextureSampleParameter2D(UVs=Subtract(A=@n1, B=@n6)) {parameterName=RT, texture=&Black}
  n9 = Max(A=Max(A=@n4.B, B=@n5.B), B=Max(A=@n7.B, B=@n8.B))
  n10 = Saturate(None=Ceil(None=@n4.B))
  n11 = Saturate(None=Ceil(None=@n5.B))
  n12 = Saturate(None=Ceil(None=@n7.B))
  n13 = Saturate(None=Ceil(None=@n8.B))
  n14 = Divide(A=Add(A=Add(A=@n4.B, B=@n5.B), B=Add(A=@n7.B, B=@n8.B)), B=Add(A=Add(A=@n10, B=@n11), B=Add(A=@n12, B=@n13))) {constB=4}
  LinearInterpolate(A=@n9, B=@n14)
  n15 = TextureSampleParameter2D() {parameterName=RT, texture=&Black}
  n16 = If(A=@n15.B, "A > B"=Add(A=@n15.R, B=If(A=@n15.B, "A > B"=@n14, "A == B"=@n9, "A < B"=@n9)), "A == B"=@n15.B, "A < B"=@n15.B)
  n17 = ScalarParameter() {parameterName=Min}
  n18 = ScalarParameter() {defaultValue=10000, parameterName=Max}
  Saturate(None=Divide(A=Subtract(A=@n16, B=@n17), B=Subtract(A=@n18, B=@n17)))
 ]
}
