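# Unlit material that prepares a render target for landscape height use in an old Landmass prototype.
# It samples the red channel of the RT texture parameter, which defaults to black, as an integer height in the -32k to 32k range.
# The PackRG8Height material function converts that value, by its name into height packed across two 8-bit channels.
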Material /Landmass/Landscape/OldPrototype_BP/Mats/Capture/M_PrepRTforLS
graph {
 properties {pixelDepthOffsetMode=PDOM_Legacy, shadingModel=MSM_Unlit, usedShadingModels=Unlit}
 nodes [MaterialFunctionCall!(Height Int -32k to 32k=TextureSampleParameter2D() {parameterName=RT, texture=&Black}.R) {materialFunction=/Landmass/Landscape/OldPrototype_BP/Mats/MF/PackRG8Height}]
}
