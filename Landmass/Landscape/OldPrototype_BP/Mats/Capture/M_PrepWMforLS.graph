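# Unlit prototype capture material that, going by its name, preps a weightmap render target for use by the Landscape.
# It samples the RT texture parameter and appends the inverted alpha to its RGB.
# A dot product of that vector with one built from the MaskChannel parameter reduces it to a single selectable channel, red by default.
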Material /Landmass/Landscape/OldPrototype_BP/Mats/Capture/M_PrepWMforLS
graph {
 properties {bUsedWithNanite=true, pixelDepthOffsetMode=PDOM_Legacy, shadingModel=MSM_Unlit, usedShadingModels=Unlit}
 nodes [
  n1 = TextureSampleParameter2D() {parameterName=RT, texture=&Black}
  n2 = VectorParameter() {parameterName=MaskChannel, defaultValue=<rgba 1 0 0 0>}
  DotProduct(A=AppendVector(A=@n1.RGB, B=OneMinus(None=@n1.A)), B=AppendVector(A=@n2.RGB, B=@n2.RGB))
 ]
}
